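# Physics asset for the M_MED_Caterpillar back accessory, defining collision and joint limits for its six-bone dynamic cape.
# The backpack root carries two capsules, one much larger and offset, while each cape bone has a single capsule.
# Constraints chain each cape bone to the one above, up to the backpack root.
# Swing limits loosen toward the tip, while twist stays tightly limited throughout.
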PhysicsAsset /BRCosmetics/Accessories/FORT_Capes/M_MED_Caterpillar/Meshes/M_MED_Caterpillar_Pack_Physics
bodies {
 backpack_root [
  {
   shapeType=Capsule
   length=14.1718
   radius=8.2405
   shapeName=backpack_root_capsule
   center=<0.0039 0.5801 -3.3845>
   rotation=<rot 0 -90.0337 -89.963>
  }
  {shapeType=Capsule, length=41.375, radius=23.9062, center=<-0.0001 23.6688 -29.3732>, rotation=<rot 0 0 5.018>}
 ]
 dyn_cape01 [{shapeType=Capsule, length=13.413, radius=5.948, shapeName=dyn_cape01_capsule, center=<5.9733 -5.2 0.0047>, rotation=<rot 0.0435 -90 0.0034>}]
 dyn_cape02 [{shapeType=Capsule, length=12.3711, radius=6.784, shapeName=dyn_cape02_capsule, center=<7.0599 -5.2 0.0119>, rotation=<rot 0.111 -90 0.0007>}]
 dyn_cape03 [{shapeType=Capsule, length=12.9304, radius=6.284, shapeName=dyn_cape03_capsule, center=<6.7182 -2.5 0.0005>, rotation=<rot 0.0814 -90 0.0784>}]
 dyn_cape04 [{
   shapeType=Capsule
   length=18.2828
   radius=4.784
   shapeName=dyn_cape04_capsule
   center=<8.7112 -0.0426 0.0068>
   rotation=<rot 0.0928 -90 0.1536>
  }]
 dyn_cape05 [{
   shapeType=Capsule
   length=17.5972
   radius=4.784
   shapeName=dyn_cape05_capsule
   center=<9.9651 0.4136 -0.0004>
   rotation=<rot 0.1167 -90 0.0018>
  }]
 dyn_cape06 [{
   shapeType=Capsule
   length=15.9893
   radius=4.784
   shapeName=dyn_cape06_capsule
   center=<5.7744 0.4105 -0.0003>
   rotation=<rot 0.009 -90 0.0044>
  }]
}
constraints [
 {bone1Name=dyn_cape06, bone2Name=dyn_cape05, swing1LimitDegrees=45, swing2LimitDegrees=45, twistLimitDegrees=3}
 {bone1Name=dyn_cape05, bone2Name=dyn_cape04, swing1LimitDegrees=45, swing2LimitDegrees=45, twistLimitDegrees=3}
 {bone1Name=dyn_cape04, bone2Name=dyn_cape03, swing1LimitDegrees=45, swing2LimitDegrees=45, twistLimitDegrees=3}
 {bone1Name=dyn_cape03, bone2Name=dyn_cape02, swing1LimitDegrees=40, swing2LimitDegrees=40, twistLimitDegrees=3}
 {bone1Name=dyn_cape02, bone2Name=dyn_cape01, swing1LimitDegrees=30, swing2LimitDegrees=30, twistLimitDegrees=3}
 {bone1Name=dyn_cape01, bone2Name=backpack_root, swing1LimitDegrees=20, swing2LimitDegrees=20, twistLimitDegrees=3}
]
