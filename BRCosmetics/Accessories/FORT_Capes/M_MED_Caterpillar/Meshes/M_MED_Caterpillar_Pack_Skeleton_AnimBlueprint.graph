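# Animation Blueprint for the Caterpillar back bling (cape/pack cosmetic), adding secondary physics motion.
# The output starts from ref pose, rolls backpack_root slightly when the parachute opens, then runs RigidBody simulation relative to that bone with overridden gravity.
# Each update sets gravity through a deprecated helper, parachute alpha, and parent-mesh bone world transforms.
# Those transforms drive a Modify Bone chain saved as cached 'Pose', which never reaches the output.
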AnimBlueprint /BRCosmetics/Accessories/FORT_Capes/M_MED_Caterpillar/Meshes/M_MED_Caterpillar_Pack_Skeleton_AnimBlueprint
target_skeleton=&/M_MED_Caterpillar_Pack_Skeleton
graphs [
 AnimGraph {nodes [
   "Transform (Modify) Bone - Bone: spine_01"(Alpha=1, Translation="0,0,0", Rotation="0,0,0", Scale="1,1,1") {
    rotationMode=BMM_Replace
    rotationSpace=BCS_WorldSpace
    scaleSpace=BCS_BoneSpace
    translationMode=BMM_Replace
    translationSpace=BCS_WorldSpace
    alphaBoolBlend {blendOption=Linear}
    boneToModify {boneName=spine_01}
   }
   n1 = "Transform (Modify) Bone - Bone: backpack_root"!(
    ComponentPose=Local To Component(LocalPose=Local Space Ref Pose())
    Alpha=Get Parachute Opened Alpha() -> (Parachute Opened Alpha:double)
    Translation="0,0,0"
    Rotation="0,0,10"
   ) -> (Pose) {
    rotationMode=BMM_Additive
    rotationSpace=BCS_BoneSpace
    translationSpace=BCS_BoneSpace
    alphaBoolBlend {blendOption=Linear}
    boneToModify {boneName=backpack_root}
    rotation {roll=10}
   }
   n2 = RigidBody!(ComponentPose=@n1, Alpha, OverrideWorldGravity:Vector=Get Gravity Override(), ExternalForce) -> (Pose) {
    bClampLinearTranslationLimitToRefPose=true
    bOverrideWorldGravity=true
    overridePhysicsAsset=&/M_MED_Caterpillar_Pack_Physics
    simulationSpace=BaseBoneSpace
    alphaBoolBlend {blendOption=Linear}
    baseBoneRef {boneName=backpack_root}
    componentAppliedLinearAccClamp=<3000 3000 3000>
    componentLinearAccScale {z=0.5}
    componentLinearVelScale {z=0.5}
    overrideWorldGravity {z=-980}
    simSpaceSettings {
     maxAngularAcceleration=500
     maxAngularVelocity=3
     maxLinearAcceleration=700
     maxLinearVelocity=500
     velocityScaleZ=0
     worldAlpha=1
     externalLinearDragV=<1 1>
    }
   }
   Output Pose(Result=Component To Local(ComponentPose=@n2)) {layerGroup=None}
   n3 = "Transform (Modify) Bone - Bone: body_root"(
    ComponentPose=Local To Component(LocalPose=Use cached pose 'initCache'())
    Alpha=Get isEmotingAlpha() -> (isEmotingAlpha:double)
    Translation=Get bodyRoot_tran() -> (bodyRoot_tran:Vector)
    Rotation=Get bodyRoot_rot() -> (bodyRoot_rot:Rotator)
    Scale="1,1,1"
   ) {
    rotationMode=BMM_Replace
    rotationSpace=BCS_WorldSpace
    scaleSpace=BCS_BoneSpace
    translationMode=BMM_Replace
    translationSpace=BCS_WorldSpace
    alphaBoolBlend {blendOption=Linear}
    alphaScaleBiasClamp {inRange {max=0, min=-40}, outRange {max=0, min=1}}
    boneToModify {boneName=body_root}
   }
   n4 = "Transform (Modify) Bone - Bone: pelvis"(
    ComponentPose=@n3
    Alpha=1
    Translation=Get pelvis_tran() -> (pelvis_tran:Vector)
    Rotation=Get pelvis_rot() -> (pelvis_rot:Rotator)
    Scale="1,1,1"
   ) {
    rotationMode=BMM_Replace
    rotationSpace=BCS_WorldSpace
    scaleSpace=BCS_BoneSpace
    translationMode=BMM_Replace
    translationSpace=BCS_WorldSpace
    alphaBoolBlend {blendOption=Linear}
    alphaScaleBiasClamp {inRange {max=0, min=-40}, outRange {max=0, min=1}}
    boneToModify {boneName=pelvis}
   }
   n5 = "Transform (Modify) Bone - Bone: thigh_l"(
    ComponentPose=@n4
    Alpha=1
    Translation=Get thigh_l_tran() -> (thigh_l_tran:Vector)
    Rotation=Get thigh_l_rot() -> (thigh_l_rot:Rotator)
    Scale="1,1,1"
   ) {
    rotationMode=BMM_Replace
    rotationSpace=BCS_WorldSpace
    scaleSpace=BCS_BoneSpace
    translationMode=BMM_Replace
    translationSpace=BCS_WorldSpace
    alphaBoolBlend {blendOption=Linear}
    alphaScaleBiasClamp {inRange {max=0, min=-40}, outRange {max=0, min=1}}
    boneToModify {boneName=thigh_l}
   }
   n6 = "Transform (Modify) Bone - Bone: thigh_r"(
    ComponentPose=@n5
    Alpha=1
    Translation=Get thigh_r_tran() -> (thigh_r_tran:Vector)
    Rotation=Get thigh_r_rot() -> (thigh_r_rot:Rotator)
    Scale="1,1,1"
   ) {
    rotationMode=BMM_Replace
    rotationSpace=BCS_WorldSpace
    scaleSpace=BCS_BoneSpace
    translationMode=BMM_Replace
    translationSpace=BCS_WorldSpace
    alphaBoolBlend {blendOption=Linear}
    alphaScaleBiasClamp {inRange {max=0, min=-40}, outRange {max=0, min=1}}
    boneToModify {boneName=thigh_r}
   }
   n7 = "Transform (Modify) Bone - Bone: calf_l"(
    ComponentPose=@n6
    Alpha=1
    Translation=Get calf_l_tran() -> (calf_l_tran:Vector)
    Rotation=Get calf_l_rot() -> (calf_l_rot:Rotator)
    Scale="1,1,1"
   ) {
    rotationMode=BMM_Replace
    rotationSpace=BCS_WorldSpace
    scaleSpace=BCS_BoneSpace
    translationMode=BMM_Replace
    translationSpace=BCS_WorldSpace
    alphaBoolBlend {blendOption=Linear}
    alphaScaleBiasClamp {inRange {max=0, min=-40}, outRange {max=0, min=1}}
    boneToModify {boneName=calf_l}
   }
   n8 = "Transform (Modify) Bone - Bone: calf_r"(
    ComponentPose=@n7
    Alpha=1
    Translation=Get calf_r_tran() -> (calf_r_tran:Vector)
    Rotation=Get calf_r_rot() -> (calf_r_rot:Rotator)
    Scale="1,1,1"
   ) {
    rotationMode=BMM_Replace
    rotationSpace=BCS_WorldSpace
    scaleSpace=BCS_BoneSpace
    translationMode=BMM_Replace
    translationSpace=BCS_WorldSpace
    alphaBoolBlend {blendOption=Linear}
    alphaScaleBiasClamp {inRange {max=0, min=-40}, outRange {max=0, min=1}}
    boneToModify {boneName=calf_r}
   }
   n9 = "Transform (Modify) Bone - Bone: spine_02"(
    ComponentPose=@n8
    Alpha=1
    Translation=Get Spine 02 Tran() -> (Spine 02 Tran:Vector)
    Rotation=Get Spine 02 Rot() -> (Spine 02 Rot:Rotator)
    Scale="1,1,1"
   ) {
    rotationMode=BMM_Replace
    rotationSpace=BCS_WorldSpace
    scaleSpace=BCS_BoneSpace
    translationMode=BMM_Replace
    translationSpace=BCS_WorldSpace
    alphaBoolBlend {blendOption=Linear}
    alphaScaleBiasClamp {inRange {max=0, min=-40}, outRange {max=0, min=1}}
    boneToModify {boneName=spine_02}
   }
   n10 = "Transform (Modify) Bone - Bone: spine_03"(
    ComponentPose=@n9
    Alpha=1
    Translation=Get Spine 03 Tran() -> (Spine 03 Tran:Vector)
    Rotation=Get Spine 03 Rot() -> (Spine 03 Rot:Rotator)
    Scale="1,1,1"
   ) {
    rotationMode=BMM_Replace
    rotationSpace=BCS_WorldSpace
    scaleSpace=BCS_BoneSpace
    translationMode=BMM_Replace
    translationSpace=BCS_WorldSpace
    alphaBoolBlend {blendOption=Linear}
    boneToModify {boneName=spine_03}
   }
   n11 = "Transform (Modify) Bone - Bone: spine_04"(
    ComponentPose=@n10
    Alpha=1
    Translation=Get Spine 04 Tran() -> (Spine 04 Tran:Vector)
    Rotation=Get Spine 04 Rot() -> (Spine 04 Rot:Rotator)
    Scale="1,1,1"
   ) {
    rotationMode=BMM_Replace
    rotationSpace=BCS_WorldSpace
    scaleSpace=BCS_BoneSpace
    translationMode=BMM_Replace
    translationSpace=BCS_WorldSpace
    alphaBoolBlend {blendOption=Linear}
    boneToModify {boneName=spine_04}
   }
   n12 = "Transform (Modify) Bone - Bone: spine_05"(
    ComponentPose=@n11
    Alpha=1
    Translation=Get Spine05 Tran() -> (Spine05 Tran:Vector)
    Rotation=Get Spine 05 Rot() -> (Spine 05 Rot:Rotator)
    Scale="1,1,1"
   ) {
    rotationMode=BMM_Replace
    rotationSpace=BCS_WorldSpace
    scaleSpace=BCS_BoneSpace
    translationMode=BMM_Replace
    translationSpace=BCS_WorldSpace
    alphaBoolBlend {blendOption=Linear}
    alphaScaleBiasClamp {inRange {max=0, min=-40}, outRange {max=0, min=1}}
    boneToModify {boneName=spine_05}
   }
   n13 = "Transform (Modify) Bone - Bone: upperarm_l"(
    ComponentPose=@n12
    Alpha=1
    Translation=Get upperarm l Tran() -> (upperarm l Tran:Vector)
    Rotation=Get upperarm l Rot() -> (upperarm l Rot:Rotator)
    Scale="1,1,1"
   ) {
    rotationMode=BMM_Replace
    rotationSpace=BCS_WorldSpace
    scaleSpace=BCS_BoneSpace
    translationMode=BMM_Replace
    translationSpace=BCS_WorldSpace
    alphaBoolBlend {blendOption=Linear}
    alphaScaleBiasClamp {inRange {max=0, min=-40}, outRange {max=0, min=1}}
    boneToModify {boneName=upperarm_l}
   }
   n14 = "Transform (Modify) Bone - Bone: upperarm_r"(
    ComponentPose=@n13
    Alpha=1
    Translation=Get upperarm r Tran() -> (upperarm r Tran:Vector)
    Rotation=Get upperarm r Rot() -> (upperarm r Rot:Rotator)
    Scale="1,1,1"
   ) {
    rotationMode=BMM_Replace
    rotationSpace=BCS_WorldSpace
    scaleSpace=BCS_BoneSpace
    translationMode=BMM_Replace
    translationSpace=BCS_WorldSpace
    alphaBoolBlend {blendOption=Linear}
    alphaScaleBiasClamp {inRange {max=0, min=-40}, outRange {max=0, min=1}}
    boneToModify {boneName=upperarm_r}
   }
   Save cached pose 'Pose'(Pose=Component To Local(ComponentPose=@n14)) {cacheName=Pose}
   Save cached pose 'initCache'() {cacheName=initCache}
   Use cached pose 'Pose'()
  ]}
 EventGraph {nodes [
   n1 = Event BlueprintUpdateAnimation()
   > n2 = Sequence!(execute) -> (then_0, then_1, then_2:Exec)
   > Set Body_Mesh(Body_Mesh=GetOwnerSkeletalMeshComponent())
   n3 = GetPartSkeletalMeshComponent()
   > Is Valid(InputObject=@n3)
   n4 = GetAttachParent!(self:Scene Component Object Reference=@n3) -> (ReturnValue:Scene Component Object Reference)
   > Is Valid(InputObject=@n4)
   > n5 = Sequence!(execute) -> (
    then_0
    then_1
    then_2
    then_3:Exec
    then_4:Exec
    then_5:Exec
    then_6:Exec
    then_7:Exec
    then_8:Exec
    then_9:Exec
    then_10:Exec
    then_11:Exec
   )
   n6 = BreakTransform(InTransform=GetSocketTransform(self=@n4, InSocketName=spine_05, TransformSpace=RTS_World))
   Set Spine05 Tran(execute=@n5.then_6, Spine05 Tran:Vector=@n6) -> (Output_Get:Vector)
   > Set Spine 05 Rot(Spine 05 Rot:Rotator=@n6.Rotation) -> (Output_Get:Rotator)
   n7 = BreakTransform(InTransform=GetSocketTransform(self=@n4, InSocketName=spine_02, TransformSpace=RTS_World))
   Set Spine 02 Tran(execute=@n5.then_7, Spine 02 Tran:Vector=@n7) -> (Output_Get:Vector)
   > Set Spine 02 Rot(Spine 02 Rot:Rotator=@n7.Rotation) -> (Output_Get:Rotator)
   n8 = BreakTransform(InTransform=GetSocketTransform(self=@n4, InSocketName=spine_03, TransformSpace=RTS_World))
   Set Spine 03 Tran(execute=@n5.then_8, Spine 03 Tran:Vector=@n8) -> (Output_Get:Vector)
   > Set Spine 03 Rot(Spine 03 Rot:Rotator=@n8.Rotation) -> (Output_Get:Rotator)
   n9 = BreakTransform(InTransform=GetSocketTransform(self=@n4, InSocketName=spine_04, TransformSpace=RTS_World))
   Set Spine 04 Tran(execute=@n5.then_9, Spine 04 Tran:Vector=@n9) -> (Output_Get:Vector)
   > Set Spine 04 Rot(Spine 04 Rot:Rotator=@n9.Rotation) -> (Output_Get:Rotator)
   n10 = BreakTransform(InTransform=GetSocketTransform(self=@n4, InSocketName=upperarm_l, TransformSpace=RTS_World))
   Set upperarm l Tran(execute=@n5.then_10, upperarm l Tran:Vector=@n10) -> (Output_Get:Vector)
   > Set upperarm l Rot(upperarm l Rot:Rotator=@n10.Rotation) -> (Output_Get:Rotator)
   n11 = BreakTransform(InTransform=GetSocketTransform(self=@n4, InSocketName=upperarm_r, TransformSpace=RTS_World))
   Set upperarm r Tran(execute=@n5.then_11, upperarm r Tran:Vector=@n11) -> (Output_Get:Vector)
   > Set upperarm r Rot(upperarm r Rot:Rotator=@n11.Rotation) -> (Output_Get:Rotator)
   n12 = BreakTransform(InTransform=GetSocketTransform(self=@n4, InSocketName=pelvis, TransformSpace=RTS_World))
   Set pelvis_tran(execute=@n5.then_0, pelvis_tran:Vector=@n12) -> (Output_Get:Vector)
   > Set pelvis_rot(pelvis_rot:Rotator=@n12.Rotation) -> (Output_Get:Rotator)
   n13 = BreakTransform(InTransform=GetSocketTransform(self=@n4, InSocketName=thigh_l, TransformSpace=RTS_World))
   Set thigh_l_tran(execute=@n5.then_1, thigh_l_tran:Vector=@n13) -> (Output_Get:Vector)
   > Set thigh_l_rot(thigh_l_rot:Rotator=@n13.Rotation) -> (Output_Get:Rotator)
   n14 = BreakTransform(InTransform=GetSocketTransform(self=@n4, InSocketName=thigh_r, TransformSpace=RTS_World))
   Set thigh_r_tran(execute=@n5.then_2, thigh_r_tran:Vector=@n14) -> (Output_Get:Vector)
   > Set thigh_r_rot(thigh_r_rot:Rotator=@n14.Rotation) -> (Output_Get:Rotator)
   n15 = BreakTransform(InTransform=GetSocketTransform(self=@n4, InSocketName=calf_r, TransformSpace=RTS_World))
   Set calf_r_tran(execute=@n5.then_4, calf_r_tran:Vector=@n15) -> (Output_Get:Vector)
   > Set calf_r_rot(calf_r_rot:Rotator=@n15.Rotation) -> (Output_Get:Rotator)
   n16 = BreakTransform(InTransform=GetSocketTransform(self=@n4, InSocketName=calf_l, TransformSpace=RTS_World))
   Set calf_l_tran(execute=@n5.then_3, calf_l_tran:Vector=@n16) -> (Output_Get:Vector)
   > Set calf_l_rot(calf_l_rot:Rotator=@n16.Rotation) -> (Output_Get:Rotator)
   n17 = BreakTransform(InTransform=GetSocketTransform(self=@n4, InSocketName=root, TransformSpace=RTS_World))
   Set bodyRoot_tran(execute=@n5.then_5, bodyRoot_tran:Vector=@n17) -> (Output_Get:Vector)
   > Set bodyRoot_rot(bodyRoot_rot:Rotator=@n17.Rotation) -> (Output_Get:Rotator)
   n18 = Gravity Override State Function DEPRECATED(
    execute=@n2.then_1
    CharacterPartAnimInstance=Self-Reference()
    GameTimeInSeconds=@n1.DeltaTimeX
    GravityOverrideParams=Get Gravity Override Params()
   )
   > Set Gravity Override(Gravity Override=@n18)
   Set Parachute Opened Alpha(execute=@n2.then_2, Parachute Opened Alpha:double="To Float (Boolean)"(InBool=Get bIsParachuteOpened())) -> (Output_Get:double)
   BreakTransform(InTransform=GetSocketTransform(InSocketName=spine_01, TransformSpace=RTS_World))
  ]}
]
variables [
 AimOffset_0 {type=Vector}
 AimOffsetMultiplier_0 {type=number, default=-0.25}
 bBlendInBackwardPose_0 {type=boolean}
 BlendOutForksAlpha_0 {type=number}
 Body_Mesh {type=SkeletalMeshComponent}
 bodyRoot_rot {type=Rotator}
 bodyRoot_tran {type=Vector}
 calf_l_rot {type=Rotator}
 calf_l_tran {type=Vector}
 calf_r_rot {type=Rotator}
 calf_r_tran {type=Vector}
 CapeFakeVelocity_0 {type=Vector, default=<30 30 -50>}
 clothing_sim_interactor_Old {type=ClothingSimulationInteractorNv, default=null}
 FakeWindVelocity_0 {type=Vector}
 FakeWindVelocity_DBNO_Old {type=Vector}
 Gravity Override {type=Vector}
 Gravity Override Params {type=GravityOverrideParamsStruct}
 isDBNO_Old {type=boolean}
 isEmotingAlpha {type=number}
 MeshToCopy_0 {type=SkeletalMeshComponent}
 Parachute Opened Alpha {type=number}
 pelvis_rot {type=Rotator}
 pelvis_tran {type=Vector}
 Spine 02 Rot {type=Rotator}
 Spine 02 Tran {type=Vector}
 Spine 03 Rot {type=Rotator}
 Spine 03 Tran {type=Vector}
 Spine 04 Rot {type=Rotator}
 Spine 04 Tran {type=Vector}
 Spine 05 Rot {type=Rotator}
 Spine05 Tran {type=Vector}
 thigh_l_rot {type=Rotator}
 thigh_l_tran {type=Vector}
 thigh_r_rot {type=Rotator}
 thigh_r_tran {type=Vector}
 upperarm l Rot {type=Rotator}
 upperarm l Tran {type=Vector}
 upperarm r Rot {type=Rotator}
 upperarm r Tran {type=Vector}
]
